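# M_IceChunks is a two-sided, translucent lit particle material for ice chunk effects on sprites, mesh particles and beam trails.
# An IceMasks sub-UV sample is shaped by a normal-to-camera facing term that dims grazing angles.
# The opacity chain multiplies that mask by particle alpha and a near-camera fade set by CameraFade.
# A SubUV function also selects a 2x2 IceMasks frame from a rounded dynamic parameter.
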Material /Figment_S09/Effects/Fort_Effects/Materials/Ice/M_IceChunks
graph {
 properties {
  bUsedWithBeamTrails=true
  bUsedWithMeshParticles=true
  bUsedWithParticleSprites=true
  blendMode=BLEND_Translucent
  numCustomizedUVs=4
  pixelDepthOffsetMode=PDOM_Legacy
  translucencyLightingMode=TLM_VolumetricPerVertexNonDirectional
  twoSided=true
  usedShadingModels=Default Lit
 }
 nodes [
  Constant() {r=0.5}
  Constant() {r=1}
  MaterialFunctionCall!(
   Texture=TextureObject() {texture=/Game/Weapons/WeaponSkins/Elemental/Textures/IceMasks}
   UVs
   SubImages=Constant2Vector() {g=2, r=2}
   Frame=MaterialFunctionCall!(In=DynamicParameter() {defaultValue=<rgba 0 0 0 1>}.R) {materialFunction=/Engine/Functions/Engine_MaterialFunctions02/Utility/Round}.Result
  ) {materialFunction=/Engine/Functions/Engine_MaterialFunctions03/Texturing/SubUV_Function}
  n1 = Normalize(VectorInput=VertexNormalWS())
  n2 = CameraVectorWS()
  n3 = Power(Base=Abs(None=DotProduct(A=@n1, B=@n2)))
  n4 = MaterialFunctionCall(Float3=Multiply(A=@n3, B=ParticleSubUV!(UVs, Tex, Apply View MipBias) {texture=&IceMasks}.R)) {materialFunction=&BreakOutFloat3Components}
  Multiply(
   A=Multiply(A=ParticleColor().A, B=Multiply(A=@n3, B=@n4.R) {constA=1}) {desc=opacity}
   B=MaterialFunctionCall(Distance=ScalarParameter() {defaultValue=50, parameterName=CameraFade}) {materialFunction=&Near_Camera_Fade}.Result
  )
 ]
}
